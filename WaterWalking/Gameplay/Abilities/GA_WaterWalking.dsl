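; Gameplay ability that gives a Fortnite player pawn water walking while it is active.
; On activation it caches the avatar pawn, activates the configured movement mode extension and stores its runtime handle, or ends itself if the pawn cast fails.
; Ending the ability removes that extension through the stored handle before calling the parent.
; CanActivateAbility returns only the hotfix enable flag, and the Stop Sprint event is empty.
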(event Ability|EventActivateAbility
  ;[0 0.15 0.008209 0.5] ActivateAbility: Apply WaterWalking movement MME.
  (|CallParentFunction)
  (bind _outfortplayerpawn (CallFunction|GetAvatarasValidFortPlayerPawn)
    (:ValidFortPlayerPawn
      (bind _output_get (Variables|Transient|SetAvatarPlayerPawn _outfortplayerpawn))
      (bind _returnvalue (Class|FortMovementModeExtBlueprintLibrary|ActivateRegisteredMovementModeExtension (Variables|Settings|GetWaterWalkingMovementModeClass) _output_get))
      (Variables|Transient|SetWaterWalkingMovementModeRuntimeHandle _returnvalue))
    (:AvatarCastFailed
      (Ability|EndAbility))))

(event Ability|EventOnEndAbility
  (params (bWasCancelled :type Boolean))
  ;[0.15 0.09205 0.01604 0.5] EndAbility: Clean up.
  (Utilities|FlowControl|Sequence
    (:then_0
      (Class|FortMovementModeExtBlueprintLibrary|EndMovementModeExtensiononActor (Variables|Transient|GetWaterWalkingMovementModeRuntimeHandle) (Variables|Transient|GetAvatarPlayerPawn)))
    (:then_1
      (|CallParentFunction bWasCancelled))))

(event Custom|Stop%20Sprint)

(fn CanActivateAbility
  (params (ActorInfo :type Gameplay%20Ability%20Actor%20Info%20Structure) (Handle :type Gameplay%20Ability%20Spec%20Handle%20Structure))
  (return 0 (ScalableFloat|GetValueasBool (Variables|Hotfix|GetHotfix_WaterWalkingEnable))))
