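; Water-walking detection ability variant that allows walking on ocean water, extending a parent detection ability.
; Its IsInAllowableWater override calls the parent check, always reports disallowed water as false, and counts either disallowed or allowed water as allowed.
; The not-in-water result passes through unchanged.
; EventBlockWalkingOnOcean is overridden with an empty body, and the activate and end events are only disabled placeholders.
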(event Ability|EventActivateAbility
  (|CallParentFunction) ; This node is disabled and will not be called.\nDrag off pins to build functionality.
)

(event Ability|EventOnEndAbility
  (params (bWasCancelled :type Boolean))
  (|CallParentFunction bWasCancelled) ; This node is disabled and will not be called.\nDrag off pins to build functionality.
)

(event EventBlockWalkingOnOcean)

(fn IsInAllowableWater
  (params (FortPawn :type Fort%20Pawn%20Object%20Reference))
  (bind (_isindisallowedwater _isinallowedwater _isnotinwater) (|CallParentFunction FortPawn))
  (return false (or _isindisallowedwater _isinallowedwater) _isnotinwater))
